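; Building wall Blueprint from the tech art cheat library that keeps itself invulnerable by restoring its health.
; On BeginPlay it checks AllowDamageOverlay: if true, a looping timer periodically resets health to maximum, leaving damage in place between heals.
; Otherwise it binds to the building health-changed event and resets health to maximum on every change.
; The ShouldDie override has an empty body, and the construction script is empty.
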(event EventBeginPlay
  (if (Variables|Default|GetAllowDamageOverlay)
    (Utilities|Time|SetTimerbyEvent (delegate Custom|Heal) 5.0 true)
    (else
      (Health|BindEventtoOnBuildingHealthChanged (delegate Custom|OnBuildingHealthChanged_Event)))))

(event Custom|Heal
  (Health|SetHealth (Health|GetMaxHealth)))

(event Custom|OnBuildingHealthChanged_Event
  (Health|SetHealth (Health|GetMaxHealth)))

(fn ShouldDie
  (params (Damage :type Float%20%28single-precision%29) (EventInstigator :type Controller%20Object%20Reference) (DamageCauser :type Actor%20Object%20Reference) (bIsTest :type Boolean) (EffectContext :type Gameplay%20Effect%20Context%20Handle%20Structure%20%28by%20ref%29)))

(fn ConstructionScript)
